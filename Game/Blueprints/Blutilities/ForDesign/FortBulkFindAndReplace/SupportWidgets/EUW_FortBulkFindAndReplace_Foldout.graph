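# Editor utility widget for a collapsible foldout header, a support widget of the Fort Bulk Find And Replace design tool.
# Clicking ToggleButton toggles an assigned ContentsWidget between Visible and Collapsed, then Refresh rotates the arrow glyph to match.
# PreConstruct applies the header text, default open state, font size and button opacity.
# Open state is read from the contents' visibility, and macros rebuild the button brushes and fonts to apply opacity and size.
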EditorUtilityWidgetBlueprint /Game/Blueprints/Blutilities/ForDesign/FortBulkFindAndReplace/SupportWidgets/EUW_FortBulkFindAndReplace_Foldout
graphs [
 EventGraph {nodes [
   "On Clicked (ToggleButton)"() {is=title K2Node_ComponentBoundEvent}
   > Set Foldout Open!(execute, self, Open:bool=NOT Boolean(A=Is Foldout Open!() -> (Is Open:bool))) -> (then) {is=title K2Node_CallFunction}
   Event PreConstruct()
   > Set Title!(execute, self, Title Text:Text="Get Header Text (Default)"() -> ("Header Text (Default)":Text)) -> (then) {is=title K2Node_CallFunction}
   > Set Button Opacity!(execute:Exec) -> (then:Exec) {is=title K2Node_MacroInstance}
   > Set Foldout Open(Open=Get Default Open State() -> (Default Open State:bool))
   > Set Font Size!(
    execute:Exec
    Text Widget:Common Text Object Reference=Get TitleTextBlock() -> (TitleTextBlock:Common Text Object Reference)
    Size:int=Get Font Size() -> (Font Size:int=0)
   ) -> (then:Exec) {is=title K2Node_MacroInstance}
   > Set Font Size(Text Widget=Get ArrowSymbol() -> (ArrowSymbol:Common Text Object Reference), Size=Get Font Size() -> (Font Size=0))
   Event Construct()
   > Refresh()
  ]}
 Is Foldout Open {nodes [
   Is Foldout Open() {is=title K2Node_FunctionEntry}
   > n1 = Is Valid(InputObject=Get ContentsWidget() -> (ContentsWidget:Widget Object Reference))
   Return Node!(execute=@n1.Is Not Valid, Is Open:bool=Get Default Open State())
   > Return Node(Is Open=IsVisible(self=Get ContentsWidget()))
  ]}
 Refresh {nodes [
   Refresh() {is=title K2Node_FunctionEntry}
   > n1 = Branch(Condition=Is Foldout Open() {is=title K2Node_CallFunction})
   > Set RenderTransform!(
    execute
    RenderTransform:Widget Transform Structure=Make Widget Transform!(Translation:Vector 2D Structure=(X=0,Y=0), Scale:Vector 2D Structure=(X=1,Y=1), Shear:Vector 2D Structure=(X=0,Y=0), Angle:float=180) -> (WidgetTransform:Widget Transform Structure) {is=title K2Node_MakeStruct}
    self:Widget Object Reference=Get ArrowSymbol()
   ) -> (then, Output_Get:Widget Transform Structure)
   Set RenderTransform(
    execute=@n1.else
    RenderTransform=Make Widget Transform(Translation=(X=0,Y=0), Scale=(X=1,Y=1), Shear=(X=0,Y=0), Angle=0)
    self=Get ArrowSymbol()
   )
  ]}
 Set Button Opacity {nodes [
   Inputs()
   n1 = Break Button Style!(ButtonStyle:"Button Style Structure (by ref)"=Get WidgetStyle!(self:Button Object Reference=Get ToggleButton() -> (ToggleButton:Button Object Reference)) -> (WidgetStyle:Button Style Structure)) -> (
    Normal:Slate Brush Structure
    Hovered:Slate Brush Structure
    Pressed:Slate Brush Structure
    Disabled:Slate Brush Structure
    NormalForeground:Slate Color Structure
    HoveredForeground:Slate Color Structure
    PressedForeground:Slate Color Structure
    DisabledForeground:Slate Color Structure
    NormalPadding:Margin Structure
    PressedPadding:Margin Structure
    PressedSlateSound:Slate Sound Structure
    ClickedSlateSound:Slate Sound Structure
    HoveredSlateSound:Slate Sound Structure
   ) {is=title K2Node_BreakStruct}
   > SetStyle!(
    execute
    self:Button Object Reference=Get ToggleButton()
    InStyle:"Button Style Structure (by ref)"=Make Button Style!(
     Normal:Slate Brush Structure=Set Button Style Opacity!(InputPin:Slate Brush Structure=@n1) -> (OutputPin:Slate Brush Structure) {is=title K2Node_MacroInstance}
     Hovered:Slate Brush Structure=Set Button Style Opacity(InputPin=@n1.Hovered)
     Pressed:Slate Brush Structure=Set Button Style Opacity(InputPin=@n1.Pressed)
     Disabled:Slate Brush Structure=Set Button Style Opacity(InputPin=@n1.Disabled)
     NormalForeground:Slate Color Structure=@n1.NormalForeground
     HoveredForeground:Slate Color Structure=@n1.HoveredForeground
     PressedForeground:Slate Color Structure=@n1.PressedForeground
     DisabledForeground:Slate Color Structure=@n1.DisabledForeground
     NormalPadding:Margin Structure=@n1.NormalPadding
     PressedPadding:Margin Structure=@n1.PressedPadding
     PressedSlateSound:Slate Sound Structure=@n1.PressedSlateSound
     ClickedSlateSound:Slate Sound Structure=(ResourceObject=None)
     HoveredSlateSound:Slate Sound Structure=@n1.HoveredSlateSound
    ) -> (ButtonStyle:Button Style Structure) {is=title K2Node_MakeStruct}
   ) -> (then)
   > Outputs!(then:Exec)
  ]}
 Set Button Style Opacity {nodes [
   n1 = Break Slate Brush!(SlateBrush:"Slate Brush Structure (by ref)"=Inputs!() -> (InputPin:Slate Brush Structure)) -> (
    TintColor:Slate Color Structure
    DrawAs:ESlateBrushDrawType Enum
    Tiling:ESlateBrushTileType Enum
    Mirroring:ESlateBrushMirrorType Enum
    ImageSize:Vector 2D Structure
    Margin:Margin Structure
    ResourceObject:Object Reference
    OutlineSettings:Slate Brush Outline Settings Structure
   ) {is=title K2Node_BreakStruct}
   n2 = Break Slate Color!(SlateColor:"Slate Color Structure (by ref)"=@n1) -> (SpecifiedColor:Linear Color Structure, ColorUseRule:ESlateColorStylingMode Enum) {is=title K2Node_BreakStruct}
   n3 = Break Linear Color!(LinearColor:"Linear Color Structure (by ref)"=@n2) -> (R:float=~, G:float=~, B:float=~, A:float=~) {is=title K2Node_BreakStruct}
   n4 = Make Linear Color!(R:float=@n3, G:float=@n3.G, B:float=@n3.B, A:float=Get Button Opacity() -> (Button Opacity:double)) -> (LinearColor:Linear Color Structure) {is=title K2Node_MakeStruct}
   Outputs!(OutputPin:Slate Brush Structure=Make Slate Brush!(
     TintColor:Slate Color Structure=Make Slate Color(SpecifiedColor=@n4, ColorUseRule=@n2.ColorUseRule)
     DrawAs:ESlateBrushDrawType Enum=@n1.DrawAs
     Tiling:ESlateBrushTileType Enum=@n1.Tiling
     Mirroring:ESlateBrushMirrorType Enum=@n1.Mirroring
     ImageSize:Vector 2D Structure=@n1.ImageSize
     Margin:Margin Structure=@n1.Margin
     ResourceObject:Object Reference=@n1.ResourceObject
     OutlineSettings:Slate Brush Outline Settings Structure=@n1.OutlineSettings
    ) -> (SlateBrush:Slate Brush Structure) {is=title K2Node_MakeStruct})
  ]}
 Set Foldout Open {nodes [
   Set Foldout Open!() -> (then, Open:bool=~) {is=title K2Node_FunctionEntry}
   > n1 = Sequence()
   > Is Valid(InputObject=Get ContentsWidget())
   > n2 = Branch(Condition=Get Open() -> (Open:bool))
   n3 = Get ContentsWidget()
   > SetVisibility(self=@n3, InVisibility=Visible)
   SetVisibility(execute=@n2.else, self=@n3, InVisibility=Collapsed)
   Refresh(execute=@n1.then_1) {is=title K2Node_CallFunction}
  ]}
 Set Font Size {nodes [
   n1 = Inputs!() -> (execute:Exec, Text Widget:Common Text Object Reference, Size:int)
   n2 = Get Font!(self:Text Object Reference=@n1) -> (Font:Slate Font Info Structure)
   > n3 = Branch(Condition="Equal (Float)"(
     A=Break Slate Font Info!(SlateFontInfo:"Slate Font Info Structure (by ref)"=@n2) -> (
      FontObject:Object Reference
      FontMaterial:Object Reference
      OutlineSettings:Font Outline Settings Structure
      TypefaceFontName:Name
      Size:float=~
      LetterSpacing:int
      SkewAmount:float=~
      bForceMonospaced:bool=~
      bMaterialIsStencil:bool=~
      MonospacedWidth:float=~
     ) {is=title K2Node_BreakStruct}.Size
     B:int=@n1.Size
    ))
   Outputs!(then:Exec=[@n3.then, @n5.then])
   n4 = Set members in Slate Font Info!(
    execute:Exec=@n3.else
    StructRef:"Slate Font Info Structure (by ref)"=Copy!(Item:Slate Font Info Structure=Get Font(self=@n1)) -> (ReturnValue:Slate Font Info Structure) {is=title K2Node_Copy}
    Size:float="To Float (Integer)"(InInt=@n1.Size)
   ) -> (then:Exec, StructOut:"Slate Font Info Structure (by ref)") {is=title K2Node_SetFieldsInStruct}
   > n5 = Set Font!(execute, Font:Slate Font Info Structure=@n4, self:Text Object Reference=@n1) -> (then, Output_Get:Slate Font Info Structure)
  ]}
 Set Title {nodes [
   n1 = Set Title!() -> (then, Title Text:Text) {is=title K2Node_FunctionEntry}
   > Set Text(Text=@n1, self=Get TitleTextBlock())
  ]}
]
variables [
 Button Opacity {type=number, default=0.25}
 ContentsWidget {type=Widget, default=null}
 Default Open State {type=boolean, default=true}
 Font Size {type=integer, default=24}
 "Header Text (Default)" {type=string, default=Foldout Header}
]
widgets {
 info {rootWidgetClass=&CanvasPanel, widgetCount=8}
 widgets [
  CanvasPanel_0 {widgetClassPath=&CanvasPanel, widget=CanvasPanel_0}
  ToggleButton {
   widgetClassPath=&Button
   parent=CanvasPanel_0
   widget=ToggleButton
   properties {widgetStyle {
     clickedSlateSound {resourceObject=null}
     disabled {
      bIsDynamicallyLoaded=false
      drawAs=RoundedBox
      imageType=NoImage
      mirroring=NoMirror
      resourceName=None
      resourceObject=null
      tiling=NoTile
      imageSize=<0 0>
      margin=<ltrb 0 0 0 0>
      outlineSettings {
       bUseBrushTransparency=false
       roundingType=FixedRadius
       width=1
       color {colorUseRule=UseColor_ColorTable, specifiedColor=<rgba 0 0 0 0>}
       cornerRadii=<4 4 4 4>
      }
      tintColor {colorUseRule=UseColor_Specified, specifiedColor=<rgba 0.0395 0.0395 0.0395 1>}
      uVRegion {bIsValid=false, max=<0 0>, min=<0 0>}
     }
     disabledForeground {colorUseRule=UseColor_ColorTable, specifiedColor=<rgba 0 0 0 0>}
     hovered {
      bIsDynamicallyLoaded=false
      drawAs=RoundedBox
      imageType=NoImage
      mirroring=NoMirror
      resourceName=None
      resourceObject=null
      tiling=NoTile
      imageSize=<32 32>
      margin=<ltrb 0 0 0 0>
      outlineSettings {
       bUseBrushTransparency=false
       roundingType=FixedRadius
       width=1
       color {colorUseRule=UseColor_ColorTable, specifiedColor=<rgba 0 0 0 0>}
       cornerRadii=<4 4 4 4>
      }
      tintColor {colorUseRule=UseColor_Specified, specifiedColor=<rgba 0.0953 0.0953 0.0953 1>}
      uVRegion {bIsValid=false, max=<0 0>, min=<0 0>}
     }
     hoveredForeground {colorUseRule=UseColor_ColorTable, specifiedColor=<rgba 0 0 0 0>}
     hoveredSlateSound {resourceObject=null}
     normal {
      bIsDynamicallyLoaded=false
      drawAs=RoundedBox
      imageType=NoImage
      mirroring=NoMirror
      resourceName=None
      resourceObject=null
      tiling=NoTile
      imageSize=<32 32>
      margin=<ltrb 0 0 0 0>
      outlineSettings {
       bUseBrushTransparency=false
       roundingType=FixedRadius
       width=1
       color {colorUseRule=UseColor_ColorTable, specifiedColor=<rgba 0 0 0 0>}
       cornerRadii=<4 4 4 4>
      }
      tintColor {colorUseRule=UseColor_Specified, specifiedColor=<rgba 0.0395 0.0395 0.0395 1>}
      uVRegion {bIsValid=false, max=<0 0>, min=<0 0>}
     }
     normalForeground {colorUseRule=UseColor_ColorTable, specifiedColor=<rgba 0 0 0 0>}
     normalPadding=<ltrb 12 1.5 12 1.5>
     pressed {
      bIsDynamicallyLoaded=false
      drawAs=RoundedBox
      imageType=NoImage
      mirroring=NoMirror
      resourceName=None
      resourceObject=null
      tiling=NoTile
      imageSize=<32 32>
      margin=<ltrb 0 0 0 0>
      outlineSettings {
       bUseBrushTransparency=false
       roundingType=FixedRadius
       width=1
       color {colorUseRule=UseColor_ColorTable, specifiedColor=<rgba 0 0 0 0>}
       cornerRadii=<4 4 4 4>
      }
      tintColor {colorUseRule=UseColor_Specified, specifiedColor=<rgba 0.0284 0.0284 0.0284 1>}
      uVRegion {bIsValid=false, max=<0 0>, min=<0 0>}
     }
     pressedForeground {colorUseRule=UseColor_ColorTable, specifiedColor=<rgba 0 0 0 0>}
     pressedPadding=<ltrb 12 2.5 12 0.5>
     pressedSlateSound {resourceObject=null}
    }}
   slot {
    class=CanvasPanelSlot
    properties {bAutoSize=true, layoutData {alignment=<0 0>, anchors {maximum=<1 1>, minimum=<0 0>}, offsets=<ltrb 0 0 0 0>}}
   }
  }
  HorizontalBox_83 {
   widgetClassPath=&HorizontalBox
   parent=ToggleButton
   widget=HorizontalBox_83
   slot {class=ButtonSlot, properties {horizontalAlignment=HAlign_Fill}}
  }
  PaddingLeft {
   widgetClassPath=&Image
   parent=HorizontalBox_83
   widget=PaddingLeft
   properties {brush {
     bIsDynamicallyLoaded=false
     drawAs=RoundedBox
     imageType=NoImage
     mirroring=NoMirror
     resourceName=None
     resourceObject=null
     tiling=NoTile
     imageSize=<32 4>
     margin=<ltrb 0 0 0 0>
     outlineSettings {
      bUseBrushTransparency=false
      roundingType=HalfHeightRadius
      width=0
      color {colorUseRule=UseColor_Specified, specifiedColor=<rgba 0 0 0 0>}
      cornerRadii=<0 0 0 0>
     }
     tintColor {colorUseRule=UseColor_Specified, specifiedColor=<rgba 1 1 1 0.25>}
     uVRegion {bIsValid=false, max=<0 0>, min=<0 0>}
    }}
   slot {class=HorizontalBoxSlot, properties {verticalAlignment=VAlign_Center, size {sizeRule=Fill, value=1}}}
  }
  FoldoutHeader {
   widgetClassPath=&HorizontalBox
   parent=HorizontalBox_83
   widget=FoldoutHeader
   slot {class=HorizontalBoxSlot, properties {horizontalAlignment=HAlign_Center, verticalAlignment=VAlign_Center, padding=<ltrb 4 2 4 2>}}
  }
  TitleTextBlock {
   widgetClassPath=&CommonTextBlock
   bIsVariable=true
   parent=FoldoutHeader
   widget=TitleTextBlock
   properties {
    justification=Center
    text=Foldout Title
    font {
     bForceMonospaced=false
     bMaterialIsStencil=false
     fontMaterial=null
     fontObject=&NotoSans
     letterSpacing=0
     monospacedWidth=1
     size=24
     skewAmount=0
     typefaceFontName=NotoSans-Regular
     outlineSettings {
      bApplyOutlineToDropShadows=false
      bMiteredCorners=false
      bSeparateFillAlpha=false
      outlineMaterial=null
      outlineSize=0
      outlineColor=<rgba 0 0 0 1>
     }
    }
    shadowOffset=<0 0>
   }
   slot {class=HorizontalBoxSlot, properties {horizontalAlignment=HAlign_Center, padding=<ltrb 20 0 20 0>}}
  }
  ArrowSymbol {
   widgetClassPath=&CommonTextBlock
   bIsVariable=true
   parent=FoldoutHeader
   widget=ArrowSymbol
   properties {
    justification=Center
    text="▼"
    font {
     bForceMonospaced=false
     bMaterialIsStencil=false
     fontMaterial=null
     fontObject=&Roboto
     letterSpacing=0
     monospacedWidth=1
     size=24
     skewAmount=0
     typefaceFontName=Regular
     outlineSettings {
      bApplyOutlineToDropShadows=false
      bMiteredCorners=false
      bSeparateFillAlpha=false
      outlineMaterial=null
      outlineSize=0
      outlineColor=<rgba 0 0 0 1>
     }
    }
    shadowOffset=<0 0>
   }
   slot {class=HorizontalBoxSlot, properties {horizontalAlignment=HAlign_Center, verticalAlignment=VAlign_Center, padding=<ltrb 0 0 20 0>}}
  }
  PaddingRight {
   widgetClassPath=&Image
   parent=HorizontalBox_83
   widget=PaddingRight
   properties {brush {
     bIsDynamicallyLoaded=false
     drawAs=RoundedBox
     imageType=NoImage
     mirroring=NoMirror
     resourceName=None
     resourceObject=null
     tiling=NoTile
     imageSize=<32 4>
     margin=<ltrb 0 0 0 0>
     outlineSettings {
      bUseBrushTransparency=false
      roundingType=HalfHeightRadius
      width=0
      color {colorUseRule=UseColor_Specified, specifiedColor=<rgba 0 0 0 0>}
      cornerRadii=<0 0 0 0>
     }
     tintColor {colorUseRule=UseColor_Specified, specifiedColor=<rgba 1 1 1 0.25>}
     uVRegion {bIsValid=false, max=<0 0>, min=<0 0>}
    }}
   slot {class=HorizontalBoxSlot, properties {verticalAlignment=VAlign_Center, size {sizeRule=Fill, value=1}}}
  }
 ]
}
